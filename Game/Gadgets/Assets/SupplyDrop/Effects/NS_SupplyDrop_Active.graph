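# Niagara effect for the active supply drop gadget: a glowing ground marker of stacked rings, arrows, cylinders and a tilted beam.
# One emitter burst-spawns a persistent particle whose per-part position, scale, color and orientation attributes serve nine mesh renderers.
# System Timeline's normalized time drives curves that grow the parts from zero and scales the pulse waveforms.
# Engine time spins three rings at different rates, while a shared waveform pulses the arrows and center ring.
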NiagaraSystem /Game/Gadgets/Assets/SupplyDrop/Effects/NS_SupplyDrop_Active
dependencies {
 usedDataInterfaces [&NiagaraDataInterfaceCurve, &NiagaraDataInterfaceVectorCurve]
 usedDynamicInputs [/Game/Effects/Niagara/DynamicInputs/State/FN_IsExecutionStateActive]
 usedModules [&SystemState, &Timeline]
}
emitters {SupplyDrop_Widget {
  emitterUpdateScript [
   &EmitterState(Life Cycle Mode=::NewEnumerator1, Loop Behavior=::NewEnumerator1, Loop Duration Mode=::NewEnumerator1, UseLoopDelay=false)
   &SpawnBurst_Instantaneous(Use Spawn Probability=false, UseLoopCountLimit=false)
  ]
  particleSpawnScript [
   &InitializeParticle(Mesh Scale=<0.8 0.8 1>, Mesh Scale Mode=::NewEnumerator4, UsePositionOffset=false, Write Material Random=true)
   /SetVariables(
    Particles.Rings Position:NiagaraPosition=&AddVectorToPosition(Position=$Particles.Position, Vector=<0 0 20>)
    Particles.Rings Scale:Vector3f=<0.8 0.8 1>
   ) {isSetParameters=true}
   /SetVariables(
    Particles.Rings2 Position:NiagaraPosition=&AddVectorToPosition(Position=$Particles.Position, Vector=<0 0 10>)
    Particles.Rings2 Scale:Vector3f=<1.3 1.3 5>
   ) {isSetParameters=true}
   /SetVariables(
    Particles.Ring_Flat Color:LinearColor=<rgba 0.1497 0.4294 0.5923 0.3>
    Particles.Ring_Flat Position:NiagaraPosition=&AddVectorToPosition(Position=$Particles.Position, Vector=<0 0 10>)
    Particles.Ring_Flat Scale:Vector3f=<1 1 5>
   ) {isSetParameters=true}
   /SetVariables(
    Particles.Rings2 Inner Color:LinearColor=<rgba 2.1213 5.3705 10.987 1>
    Particles.Rings2 Inner Position:NiagaraPosition=&AddVectorToPosition(Position=$Particles.Position, Vector=<0 0 10>)
    Particles.Rings2 Inner Scale:Vector3f=<1 1 5>
   ) {isSetParameters=true}
   /SetVariables(
    Particles.InnerCylinder Color:LinearColor=<rgba 1 4.4738 6.6234 0.25>
    Particles.InnerCylinder Position:NiagaraPosition=&AddVectorToPosition(Position=$Particles.Position, Vector=<0 0 35>)
    Particles.InnerCylinder Scale:Vector3f=<0.5 0.5 2>
   ) {isSetParameters=true}
   /SetVariables(
    Particles.OuterCylinder Color:LinearColor=<rgba 1 3.6942 6.6234 1>
    Particles.OuterCylinder Position:NiagaraPosition=&AddVectorToPosition(Position=$Particles.Position, Vector=<0 0 10>)
    Particles.OuterCylinder Scale:Vector3f=<0.8 0.8 1.2>
   ) {isSetParameters=true}
   /SetVariables(Particles.Arrows Position:NiagaraPosition=&AddVectorToPosition(Position=$Particles.Position, Vector=<0 0 10>)) {isSetParameters=true}
   /SetVariables(
    Particles.CenterRing Position:NiagaraPosition=&AddVectorToPosition(Position=$Particles.Position, Vector=<0 0 10>)
    Particles.CenterRing Scale:Vector3f=<0.3 0.3 5>
   ) {isSetParameters=true}
   /SetVariables(
    Particles.Beam Color:LinearColor=<rgba 1.4628 5.3705 10.987 0.02>
    Particles.Beam Mesh Orientation:Quat4f=&MakeQuatFromAxisAngle(Axis=<0 1 0>, AngleInDegrees=20, NormalizeAxis=true)
    Particles.Beam Position:NiagaraPosition=&AddVectorToPosition(Position=$Particles.Position, Vector=<0 0 5>)
    Particles.Beam Scale:Vector3f=<1 1 1>
   ) {isSetParameters=true}
  ]
  particleUpdateScript [
   &ParticleState(Kill Particles When Lifetime Has Elapsed=false, Let Infinitely Lived Particles Die When Emitter Deactivates=true)
   /SetVariables(Particles.Rings Color:LinearColor=&MultiplyLinearColor(
     Linear Color=<rgba 0.2215 1.9555 3.8847 1>
     Scale Factor=&MakeVector4_FromFloat(Float=&Waveform(
       Global Amplitude Scale=$System.Timeline.NormalizedTime
       "Waveform [1]"=::NewEnumerator7
       "[1] Frequency"=0.333
       "[1] Amplitude Min/Max"=<1 0>
      ))
    )) {isSetParameters=true}
   /SetVariables(Particles.Rings2 Mesh Orientation:Quat4f=&MakeQuatFromAxisAngle(
     Axis=<0 0 1>
     AngleInDegrees=&Lerp_Float(A=0, B=360, Alpha=&Multiply_Float(A=$Engine.Time, B=0.125))
     NormalizeAxis=true
    )) {isSetParameters=true}
   /SetVariables(Particles.Rings2 Inner Mesh Orientation:Quat4f=&MakeQuatFromAxisAngle(
     Axis=<0 0 1>
     AngleInDegrees=&Lerp_Float(A=0, B=360, Alpha=&Multiply_Float(A=$Engine.Time, B=0.0312))
     NormalizeAxis=true
    )) {isSetParameters=true}
   /SetVariables(Particles.Arrows Waveform Control:NiagaraFloat=&Waveform(Global Amplitude Scale=$System.Timeline.NormalizedTime, "Waveform [1]"=::NewEnumerator7, "[1] Frequency"=1, "[1] Amplitude Min/Max"=<0 1>)) {isSetParameters=true}
   /SetVariables(
    Particles.Arrows Color:LinearColor=&MakeLinearColorFromVectorAndFloat(
     "Vector (RGB)"=&MakeVectorFromLinearColorRGB(Color=<rgba 0.7657 3.7276 11.625 1>)
     "Float (Alpha)"=&FloatFromCurve(
      FloatCurve=DataInterface() {propertyValues {Curve {keys [<at 0.15 1>,<at 1 0>]}}}
      CurveIndex=$Particles.Arrows Waveform Control
      Scale Curve=$System.Timeline.NormalizedTime
     )
    )
    Particles.Arrows Scale:Vector3f=&Multiply_Vector(
     A=<1 1 0.2>
     B=&VectorFromCurve(
      VectorCurve=DataInterface() {propertyValues {
        XCurve {keys [<at 0 1.21 RCIM_Cubic>,<at 0.36 0.89 RCIM_Cubic>]}
        YCurve {keys [<at 0 1.21 RCIM_Cubic>,<at 0.36 0.89 RCIM_Cubic>]}
        ZCurve {keys [<at 0 1 RCIM_Cubic>,<at 0.36 1 RCIM_Cubic>]}
       }}
      CurveIndex=$Particles.Arrows Waveform Control
      Scale Curve=&VectorFromFloat(Value=$System.Timeline.NormalizedTime)
     )
    )
   ) {isSetParameters=true}
   /SetVariables(
    Particles.CenterRing Color:LinearColor=&MultiplyLinearColor(
     Linear Color=<rgba 0.2215 1.9555 3.8847 1>
     Scale Factor=&MakeVector4_FromFloat(Float=$Particles.Arrows Waveform Control)
    )
    Particles.CenterRing Mesh Orientation:Quat4f=&MakeQuatFromAxisAngle(
     Axis=<0 0 1>
     AngleInDegrees=&Lerp_Float(A=0, B=360, Alpha=&Multiply_Float(A=$Engine.Time, B=0.0156))
     NormalizeAxis=true
    )
   ) {isSetParameters=true}
   /SetVariables(
    Particles.Beam Scale:Vector3f=&Multiply_VectorByFloat(
     Vector=$Particles.Initial.Beam Scale
     Float=&FloatFromCurve(
      FloatCurve=DataInterface() {propertyValues {Curve {
         defaultValue=1
         keys [
          {arriveTangent=4.1785, interpMode=RCIM_Cubic, leaveTangent=4.1785, tangentMode=RCTM_User, time=0, value=0}
          {arriveTangent=1.1179, interpMode=RCIM_Cubic, leaveTangent=1.1179, tangentMode=RCTM_User, time=0.1961, value=0.8716}
          {arriveTangent=0.4092, interpMode=RCIM_Cubic, leaveTangent=0.4092, tangentMode=RCTM_User, time=0.2768, value=0.8905}
          {arriveTangent=0.0224, interpMode=RCIM_Cubic, leaveTangent=0.0224, tangentMode=RCTM_User, time=1, value=1}
         ]
        }}}
      CurveIndex=$System.Timeline.NormalizedTime
     )
    )
    Particles.CenterRing Scale:Vector3f=&Multiply_VectorByFloat(
     Vector=$Particles.Initial.CenterRing Scale
     Float=&FloatFromCurve(
      FloatCurve=DataInterface() {propertyValues {Curve {
         defaultValue=1
         keys [
          {arriveTangent=4.1785, interpMode=RCIM_Cubic, leaveTangent=4.1785, tangentMode=RCTM_User, time=0, value=0}
          {arriveTangent=2.2242, interpMode=RCIM_Cubic, leaveTangent=2.2242, tangentMode=RCTM_User, time=0.1662, value=0.6833}
          {arriveTangent=-0.0555, interpMode=RCIM_Cubic, leaveTangent=-0.0555, tangentMode=RCTM_User, time=0.2451, value=0.8226}
          {arriveTangent=0.0224, interpMode=RCIM_Cubic, leaveTangent=0.0224, tangentMode=RCTM_User, time=1, value=1}
         ]
        }}}
      CurveIndex=$System.Timeline.NormalizedTime
     )
    )
    Particles.InnerCylinder Scale:Vector3f=&Multiply_VectorByFloat(
     Vector=$Particles.Initial.InnerCylinder Scale
     Float=&FloatFromCurve(
      FloatCurve=DataInterface() {propertyValues {Curve {
         defaultValue=1
         keys [
          {arriveTangent=4.1785, interpMode=RCIM_Cubic, leaveTangent=4.1785, tangentMode=RCTM_User, time=0, value=0}
          {arriveTangent=1.1179, interpMode=RCIM_Cubic, leaveTangent=1.1179, tangentMode=RCTM_User, time=0.1648, value=0.764}
          {arriveTangent=0.4666, interpMode=RCIM_Cubic, leaveTangent=0.4666, tangentMode=RCTM_User, time=0.2364, value=0.8245}
          {arriveTangent=0.0224, interpMode=RCIM_Cubic, leaveTangent=0.0224, tangentMode=RCTM_User, time=1, value=1}
         ]
        }}}
      CurveIndex=$System.Timeline.NormalizedTime
     )
    )
    Particles.OuterCylinder Scale:Vector3f=&Multiply_VectorByFloat(
     Vector=$Particles.Initial.OuterCylinder Scale
     Float=&FloatFromCurve(
      FloatCurve=DataInterface() {propertyValues {Curve {
         defaultValue=1
         keys [
          {arriveTangent=4.1785, interpMode=RCIM_Cubic, leaveTangent=4.1785, tangentMode=RCTM_User, time=0, value=0}
          {arriveTangent=2.71, interpMode=RCIM_Cubic, leaveTangent=2.71, tangentMode=RCTM_User, time=0.1241, value=0.6564}
          {arriveTangent=1.1051, interpMode=RCIM_Cubic, leaveTangent=1.1051, time=0.2116, value=0.7964}
          {arriveTangent=0.2582, interpMode=RCIM_Cubic, leaveTangent=0.2582, time=0.2916, value=0.8416}
          {arriveTangent=0.0224, interpMode=RCIM_Cubic, leaveTangent=0.0224, tangentMode=RCTM_User, time=1, value=1}
         ]
        }}}
      CurveIndex=$System.Timeline.NormalizedTime
     )
    )
    Particles.Ring_Flat Scale:Vector3f=&Multiply_VectorByFloat(
     Vector=$Particles.Initial.Ring_Flat Scale
     Float=&FloatFromCurve(
      FloatCurve=DataInterface() {propertyValues {Curve {
         defaultValue=1
         keys [
          {arriveTangent=4.1785, interpMode=RCIM_Cubic, leaveTangent=4.1785, tangentMode=RCTM_User, time=0, value=0}
          {arriveTangent=2.71, interpMode=RCIM_Cubic, leaveTangent=2.71, tangentMode=RCTM_User, time=0.1906, value=0.448}
          {arriveTangent=1.1377, interpMode=RCIM_Cubic, leaveTangent=1.1377, time=0.2972, value=0.588}
          {arriveTangent=0.5862, interpMode=RCIM_Cubic, leaveTangent=0.5862, time=0.4302, value=0.7206}
          {arriveTangent=0.0224, interpMode=RCIM_Cubic, leaveTangent=0.0224, tangentMode=RCTM_User, time=1, value=1}
         ]
        }}}
      CurveIndex=$System.Timeline.NormalizedTime
     )
    )
    Particles.Rings Scale:Vector3f=&Multiply_VectorByFloat(
     Vector=$Particles.Initial.Rings Scale
     Float=&FloatFromCurve(
      FloatCurve=DataInterface() {propertyValues {Curve {
         defaultValue=1
         keys [
          {arriveTangent=4.1785, interpMode=RCIM_Cubic, leaveTangent=4.1785, tangentMode=RCTM_User, time=0, value=0}
          {arriveTangent=1.1179, interpMode=RCIM_Cubic, leaveTangent=1.1179, tangentMode=RCTM_User, time=0.2246, value=0.7909}
          {arriveTangent=0.7861, interpMode=RCIM_Cubic, leaveTangent=0.7861, tangentMode=RCTM_User, time=0.3189, value=0.8384}
          {arriveTangent=0.0224, interpMode=RCIM_Cubic, leaveTangent=0.0224, tangentMode=RCTM_User, time=1, value=1}
         ]
        }}}
      CurveIndex=$System.Timeline.NormalizedTime
     )
    )
    Particles.Rings2 Inner Scale:Vector3f=&Multiply_VectorByFloat(
     Vector=$Particles.Initial.Rings2 Inner Scale
     Float=&FloatFromCurve(
      FloatCurve=DataInterface() {propertyValues {Curve {
         defaultValue=1
         keys [
          {arriveTangent=4.1785, interpMode=RCIM_Cubic, leaveTangent=4.1785, tangentMode=RCTM_User, time=0, value=0}
          {arriveTangent=1.1179, interpMode=RCIM_Cubic, leaveTangent=1.1179, tangentMode=RCTM_User, time=0.1349, value=0.5758}
          {arriveTangent=0.4904, interpMode=RCIM_Cubic, leaveTangent=0.4904, time=0.2184, value=0.662}
          {arriveTangent=0.0224, interpMode=RCIM_Cubic, leaveTangent=0.0224, tangentMode=RCTM_User, time=1, value=1}
         ]
        }}}
      CurveIndex=$System.Timeline.NormalizedTime
     )
    )
    Particles.Rings2 Scale:Vector3f=&Multiply_VectorByFloat(
     Vector=$Particles.Initial.Rings2 Scale
     Float=&FloatFromCurve(
      FloatCurve=DataInterface() {propertyValues {Curve {
         defaultValue=1
         keys [
          {arriveTangent=4.1785, interpMode=RCIM_Cubic, leaveTangent=4.1785, tangentMode=RCTM_User, time=0, value=0}
          {arriveTangent=1.989, interpMode=RCIM_Cubic, leaveTangent=1.989, tangentMode=RCTM_User, time=0.2042, value=0.6228}
          {arriveTangent=0.474, interpMode=RCIM_Cubic, leaveTangent=0.474, time=0.3044, value=0.8211}
          {arriveTangent=0.0224, interpMode=RCIM_Cubic, leaveTangent=0.0224, tangentMode=RCTM_User, time=1, value=1}
         ]
        }}}
      CurveIndex=$System.Timeline.NormalizedTime
     )
    )
   ) {isSetParameters=true}
  ]
  rendererClasses [&NiagaraMeshRendererProperties]
  renderers [
   {rendererClass=&NiagaraMeshRendererProperties, rendererIndex=0}
   {rendererClass=&NiagaraMeshRendererProperties, rendererIndex=1}
   {rendererClass=&NiagaraMeshRendererProperties, rendererIndex=2}
   {rendererClass=&NiagaraMeshRendererProperties, rendererIndex=3}
   {rendererClass=&NiagaraMeshRendererProperties, rendererIndex=4}
   {rendererClass=&NiagaraMeshRendererProperties, rendererIndex=5}
   {rendererClass=&NiagaraMeshRendererProperties, rendererIndex=6}
   {rendererClass=&NiagaraMeshRendererProperties, rendererIndex=7}
   {rendererClass=&NiagaraMeshRendererProperties, rendererIndex=8}
  ]
 }}
summary {
 systemName=NS_SupplyDrop_Active
 emitters [{emitterName=SupplyDrop_Widget, rendererClasses [&NiagaraMeshRendererProperties]}]
 userVariables []
}
