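# Sound cue for the supply drop beep, set up as an Athena ambient sound with an attenuation asset named for the falling drop.
# A random node picks one of two equally weighted beep waves, feeding a modulator whose pitch range is pinned to unity.
# A concurrency override allows one instance per owner and blocks new plays while it is active, with a retrigger time also set.
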SoundCue /Game/Sounds/Foley_Loot/Containers/SupplyDrops/Fort_SupplyDrop_Beep_Cue
root=Modulator_0
nodes [
 {id=Modulator_0, children [Random_0], properties {pitchMax=1, pitchMin=1}}
 {id=Random_0, children [WavePlayer_0,WavePlayer_2], properties {bShouldExcludeFromBranchCulling=true, weights [1,1]}}
 {id=WavePlayer_0, properties {soundWaveAssetPtr=&/SupplyDrop_Beep_01}}
 {id=WavePlayer_2, properties {soundWaveAssetPtr=&/SupplyDrop_Beep_02}}
]
properties {
 attenuationSettings=&/Fort_SupplyDrop_Beep_Falling_Athena_Att
 bOverrideConcurrency=true
 maxDistance=10000
 soundClassObject=&Fort_Ambient_Athena
 volumeMultiplier=0.601
 concurrencyOverrides {bLimitToOwner=true, maxCount=1, resolutionRule=PreventNew, retriggerTime=1}
}
