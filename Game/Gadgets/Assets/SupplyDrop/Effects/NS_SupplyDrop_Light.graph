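# Supply drop beacon Niagara effect: six blinking red glows with point lights and a synced beep.
# GlowLight arranges six partitioned particles in a ring, rendering each as a sprite and offset light driven by a shared waveform.
# Distance curves fade and resize sprites and fade out lights sooner, while time of day scales sprite alpha.
# PlayAudio thresholds an identical waveform to trigger persistent beep audio at User.Volume, and both emitters are distance culled.
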NiagaraSystem /Game/Gadgets/Assets/SupplyDrop/Effects/NS_SupplyDrop_Light
dependencies {
 usedDataInterfaces [&NiagaraDataInterfaceArrayFloat3, &NiagaraDataInterfaceCurve, &NiagaraDataInterfaceAudioPlayer]
 usedDynamicInputs []
 usedModules [&SystemState]
}
emitters {
 GlowLight {
  emitterUpdateScript [
   &EmitterState(
    Enable Distance Culling=true
    Inactive Response=::NewEnumerator2
    Life Cycle Mode=::NewEnumerator1
    Loop Behavior=::NewEnumerator1
    Loop Duration Mode=::NewEnumerator1
    MaxDistance=25000
    Module.bUseMaxDistance=true
    Module.bUseMinDistance=false
    Scalability Mode=::NewEnumerator1
    UseLoopDelay=false
   )
   &SpawnBurst_Instantaneous(Spawn Count=6, Use Spawn Probability=false, UseLoopCountLimit=false)
  ]
  particleSpawnScript [
   &PartitionParticles(Number Of Partitions=6)
   &InitializeParticle(
    Color=<rgba 2 0.0615 0.0416 1>
    Lifetime=2
    Position Offset=<0 0 257>
    Sprite Size Mode=::NewEnumerator3
    Uniform Sprite Size=300
    UsePositionOffset=true
    Write Material Random=true
   )
   &ShapeLocation(
    MaskSpawnGroup=false
    Offset=&SelectVectorFromArray(
     Direct Array Index=$Particles.PartitionParticles.Partition
     Vector Selection Array=DataInterface() {propertyValues {FloatData [<75 -43 0>, <0 -90 0>, <-75 -43 0>, <-75 43 0>, <0 90 0>, <75 43 0>]}}
    )
    Offset Mode=::NewEnumerator0
    Override Seed=false
    Sphere Radius=0
    UseOwnerScale=true
   )
  ]
  particleUpdateScript [
   /SetVariables(
    Particles.DistanceFade:NiagaraFloat=/Niagara/DynamicInputs/Scalability/ScalabilityDistanceBasedFloat(
     Curve:NiagaraDataInterfaceCurve=DataInterface() {propertyValues {Curve {keys [
         {interpMode=RCIM_Cubic, tangentMode=RCTM_User, time=0, value=0.3}
         {interpMode=RCIM_Cubic, tangentMode=RCTM_User, time=0.25, value=1}
         {interpMode=RCIM_Cubic, tangentMode=RCTM_User, time=1, value=0}
        ]}}}
     CurveIndex:NiagaraFloat=$Emitter.Scalability.DistanceFraction
     Bias:NiagaraFloat=0
     Scale:NiagaraFloat=1
    )
    Particles.DistanceFadeLight:NiagaraFloat=&ScalabilityDistanceBasedFloat(
     Curve:NiagaraDataInterfaceCurve=DataInterface() {propertyValues {Curve {keys [
         {interpMode=RCIM_Cubic, tangentMode=RCTM_User, time=0, value=1}
         {interpMode=RCIM_Cubic, tangentMode=RCTM_User, time=0.25, value=1}
         {interpMode=RCIM_Cubic, tangentMode=RCTM_User, time=1, value=0}
        ]}}}
     CurveIndex:NiagaraFloat=$Emitter.Scalability.DistanceFraction
     Bias:NiagaraFloat=0
     Scale:NiagaraFloat=1
    )
    Particles.DistanceScale:NiagaraFloat=&ScalabilityDistanceBasedFloat(
     Curve:NiagaraDataInterfaceCurve=DataInterface() {propertyValues {Curve {keys [
         {interpMode=RCIM_Cubic, leaveTangentWeight=0.3647, tangentMode=RCTM_User, time=0.15, value=0.5}
         {interpMode=RCIM_Cubic, tangentMode=RCTM_User, time=0.8, value=1}
        ]}}}
     CurveIndex:NiagaraFloat=$Emitter.Scalability.DistanceFraction
     Bias:NiagaraFloat=0
     Scale:NiagaraFloat=1
    )
    Particles.WaveFormBlink:NiagaraFloat=&Waveform("Waveform [1]"=::NewEnumerator0, "[1] Frequency"=0.75, "[1] Amplitude Min/Max"=<0 1>)
   ) {isSetParameters=true}
   &ParticleState(Kill Particles When Lifetime Has Elapsed=false)
   &ScaleSpriteSize(Scale Sprite Size Mode=::NewEnumerator0, Uniform Scale Factor=$Particles.DistanceScale)
   &ScaleColor(
    Scale Alpha=&Multiply_Float(
     A=&FNTimeOfDayScalar(Night Scale=0.75, Morning Scale=0.5, Evening Scale=0.5)
     B=$Particles.WaveFormBlink
    )
    Scale RGB=&VectorFromFloat(Value=20)
    ScaleA=true
    ScaleRGB=false
   )
   &ScaleColor(Scale Alpha=&Multiply_Float(A=1, B=$Particles.DistanceFade), ScaleA=true, ScaleRGB=false) {module=ScaleColor001}
   &DynamicMaterialParameters(
    Index 1 Param 1=50
    Module.Param0WriteEnabled1=true
    Module.Param1WriteEnabled1=true
    Module.Param2WriteEnabled1=true
    Module.Param3WriteEnabled1=true
    Write Parameter Index 0=false
    Write Parameter Index 1=true
   )
   &FN_Light_Attributes(
    Color=<rgba 1 0.0516 0.0416 1>
    Create Color Attribute=true
    Create Exponent Attribute=true
    Create Position Attribute=true
    Create Radius Attribute=true
    Enable Particle Light Enable=true
    Enable Visibility Tag=true
    Light Enabled:NiagaraBool=&SetBoolByFloatComparison(A=$Particles.DistanceFadeLight, B=1, Comparison Type=::NewEnumerator4)
    Position=&AddVectorToPosition(
     Position=Linked() {linkedVariable {name=Particles.Initial.ShapeLocation.ShapePosition, type=Vector3f}}
     Vector=&SelectVectorFromArray(
      Direct Array Index=$Particles.PartitionParticles.Partition
      Vector Selection Array=DataInterface() {propertyValues {FloatData [<80 -45 0>, <0 -90 0>, <-80 -45 0>, <-80 45 0>, <0 90 0>, <80 45 0>]}}
     )
    )
    Radius=250
    Scale Color=&Multiply_Float(
     A=&Multiply_Float(
      A=1
      B=&ScalabilityDistanceBasedFloat(
       Curve:NiagaraDataInterfaceCurve=DataInterface() {propertyValues {Curve {keys [{interpMode=RCIM_Cubic, tangentMode=RCTM_User, time=0.1, value=1}, {interpMode=RCIM_Cubic, tangentMode=RCTM_User, time=0.25, value=0}]}}}
       CurveIndex:NiagaraFloat=$Emitter.Scalability.DistanceFraction
       Bias:NiagaraFloat=0
       Scale:NiagaraFloat=1
      )
     )
     B=$Particles.WaveFormBlink
    )
   )
  ]
  rendererClasses [&NiagaraSpriteRendererProperties, &NiagaraLightRendererProperties]
  renderers [
   {rendererClass=&NiagaraSpriteRendererProperties}
   {rendererClass=&NiagaraLightRendererProperties, rendererIndex=1}
  ]
 }
 PlayAudio {
  emitterSpawnScript [/SetVariables(Emitter.BlinkAudio=DataInterface() {propertyValues {
      Attenuation=null
      Concurrency=null
      MaxPlaysPerTick=10
      SoundToPlay=/Game/Sounds/Foley_Loot/Containers/SupplyDrops/Fort_SupplyDrop_Beep_Cue
      bAllowLoopingOneShotSounds=false
      bLimitPlaysPerTick=true
      bOnlyActiveDuringGameplay=false
      bStopWhenComponentIsDestroyed=true
      ConfigurationUserParameter {parameter {name=None, typeDefHandle {registeredTypeIndex=1}, varData []}}
      ParameterNames []
     }}) {isSetParameters=true}]
  emitterUpdateScript [
   &EmitterState(
    Enable Distance Culling=true
    Inactive Response=::NewEnumerator2
    Life Cycle Mode=::NewEnumerator1
    Loop Behavior=::NewEnumerator1
    Loop Duration Mode=::NewEnumerator1
    MaxDistance=25000
    Module.bUseMaxDistance=true
    Module.bUseMinDistance=false
    Scalability Mode=::NewEnumerator1
    UseLoopDelay=false
   )
   &SpawnBurst_Instantaneous(Use Spawn Probability=false, UseLoopCountLimit=false)
  ]
  particleSpawnScript [
   &InitializeParticle(
    Color=<rgba 2 0.0615 0.0416 1>
    Lifetime=2
    Position Offset=<0 0 257>
    Sprite Size Mode=::NewEnumerator3
    Uniform Sprite Size=300
    UsePositionOffset=true
    Write Material Random=true
   )
   &ShapeLocation(MaskSpawnGroup=false, Override Seed=false, Sphere Radius=0, UseOwnerScale=true)
  ]
  particleUpdateScript [
   &ParticleState(Kill Particles When Lifetime Has Elapsed=false)
   /Niagara/Modules/Audio/PlayPersistentAudio(
    Audio Player:NiagaraDataInterfaceAudioPlayer=$Emitter.BlinkAudio
    Play Audio:NiagaraBool=&SetBoolByFloatComparison(A=&Waveform("Waveform [1]"=::NewEnumerator0, "[1] Frequency"=0.75, "[1] Amplitude Min/Max"=<0 1>), B=0.2)
    Volume:NiagaraFloat=$User.Volume
   )
   /Niagara/Modules/Audio/UpdatePersistentAudio(Audio Player:NiagaraDataInterfaceAudioPlayer=$Emitter.BlinkAudio, Update Type:ENiagara_AudioParamType=::NewEnumerator5)
  ]
  rendererClasses [&NiagaraSpriteRendererProperties]
  renderers [{rendererClass=&NiagaraSpriteRendererProperties}]
 }
}
summary {
 systemName=NS_SupplyDrop_Light
 emitters [
  {emitterName=GlowLight, rendererClasses [&NiagaraSpriteRendererProperties, &NiagaraLightRendererProperties]}
  {emitterName=PlayAudio, rendererClasses [&NiagaraSpriteRendererProperties]}
 ]
 userVariables [User.Volume {type=NiagaraFloat, defaultValue=1, description=""}]
}
