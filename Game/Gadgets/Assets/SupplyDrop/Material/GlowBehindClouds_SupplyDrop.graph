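# Unlit modulate particle material for the supply drop's glow behind clouds, drawn with depth testing disabled.
# An inverted Fresnel tinted by particle color is masked by the scene-to-pixel depth difference, using both a clamped ramp and SphereMask.
# A near-camera fade scales the glow, which is squared and offset around neutral one for modulate blending.
# The Fade parameter lerps from neutral to that result; several leftover nodes feed nothing.
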Material /Game/Gadgets/Assets/SupplyDrop/Material/GlowBehindClouds_SupplyDrop
graph {
 properties {
  bDisableDepthTest=true
  bUseTranslucencyVertexFog=false
  bUsedWithMeshParticles=true
  blendMode=BLEND_Modulate
  pixelDepthOffsetMode=PDOM_Legacy
  shadingModel=MSM_Unlit
  translucencyPass=MTP_BeforeDOF
  usedShadingModels=Unlit
 }
 nodes [
  n1 = SceneDepth()
  n2 = PixelDepth()
  n3 = Constant() {desc=distance at which the bloom will show through geo, r=2500}
  n4 = Divide(A=Subtract(A=@n1.R, B=@n2.R), B=@n3)
  Clamp(None=Multiply(A=@n4) {constB=-1})
  n5 = VertexNormalWS()
  n6 = ParticleColor()
  n7 = Multiply(
   A=MaterialFunctionCall(Normal Vector=Normalize(VectorInput=@n5), Invert Fresnel=StaticBool() {value=true}, Power=Constant() {r=4}) {materialFunction=&Fresnel_Function}.Result
   B=@n6.RGB
  )
  n8 = Clamp(None=@n4)
  n9 = Subtract(A=PixelDepth().R) {constB=512, desc=starting dist from the camera}
  n10 = Constant() {desc=distance from the player, r=700}
  n11 = Multiply(A=Multiply(A=SphereMask(A=@n1.R, B=@n2.R, Radius=@n3) {hardnessPercent=0}, B=Multiply(A=@n7, B=@n8)), B=Clamp(None=Divide(A=@n9, B=@n10)))
  LinearInterpolate(B=Add(A=Multiply(A=@n11, B=@n11)), Alpha=ScalarParameter() {defaultValue=1, parameterName=Fade}) {constA=1}
  Multiply(A=@n6.RGB, B=@n6.RGB)
  Multiply()
 ]
}
