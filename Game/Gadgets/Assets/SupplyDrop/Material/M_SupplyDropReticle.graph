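# Translucent, unlit, two-sided material for the supply drop gadget's reticle, flagged for use on mesh particles, including Niagara.
# An Emissive Color parameter, defaulting to a bright HDR blue, feeds the shared MF_FX_EmissiveAdjustments function.
# A scalar Opacity parameter defaults to a low, faint value, and the graph contains no textures or animation nodes of its own.
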Material /Game/Gadgets/Assets/SupplyDrop/Material/M_SupplyDropReticle
graph {
 properties {
  bUsedWithMeshParticles=true
  bUsedWithNiagaraMeshParticles=true
  blendMode=BLEND_Translucent
  pixelDepthOffsetMode=PDOM_Legacy
  shadingModel=MSM_Unlit
  twoSided=true
  usedShadingModels=Unlit
 }
 nodes [
  MaterialFunctionCall(Emissive=VectorParameter() {parameterName=Emissive Color, defaultValue=<rgba 0 1.94 9.95 1>}.RGB) {materialFunction=&MF_FX_EmissiveAdjustments}
  ScalarParameter() {defaultValue=0.2, parameterName=Opacity}
 ]
}
