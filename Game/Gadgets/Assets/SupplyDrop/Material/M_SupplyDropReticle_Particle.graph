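# Unlit, translucent, two-sided material for mesh particles, named for the supply drop reticle.
# Particle Color RGB drives both emissive and opacity through material attributes; particle alpha is not used.
# The UseCamOffset static switch optionally adds a clamped CameraOffset world position offset, sized by the CameraOffset parameter.
# Morph-target, blue color multiply and constant nodes are left unconnected.
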Material /Game/Gadgets/Assets/SupplyDrop/Material/M_SupplyDropReticle_Particle
graph {
 properties {
  bUseMaterialAttributes=true
  bUsedWithMeshParticles=true
  bUsedWithNiagaraMeshParticles=true
  blendMode=BLEND_Translucent
  pixelDepthOffsetMode=PDOM_Legacy
  shadingModel=MSM_Unlit
  twoSided=true
  usedShadingModels=Unlit
 }
 nodes [
  Constant() {r=0.2}
  n1 = MaterialFunctionCall() {materialFunction=/Engine/Functions/Engine_MaterialFunctions02/WorldPositionOffset/StaticMeshMorphTargets}
  LinearInterpolate(A=@n1.Morph Target 1 WPO, B=@n1.Morph Target 1 WPO) {constAlpha=0}
  Multiply(A=Constant3Vector() {constant=<rgba 0 0.1943 0.995 1>}) {constB=10}
  n2 = ParticleColor()
  n3 = SetMaterialAttributes!(MaterialAttributes, Emissive Color=@n2.RGB, Opacity=@n2.RGB) {attributeSetTypes [#1, #2]}
  n4 = ScalarParameter() {defaultValue=25, parameterName=CameraOffset}
  StaticSwitchParameter(
   True=SetMaterialAttributes!(
    MaterialAttributes=@n3
    World Position Offset=MaterialFunctionCall(Offset Amount=@n4) {materialFunction=/Engine/Functions/Engine_MaterialFunctions02/WorldPositionOffset/CameraOffset}.Clamped Camera Offset
   ) {attributeSetTypes [#3]}
   False=@n3
  ) {parameterName=UseCamOffset}
 ]
}
