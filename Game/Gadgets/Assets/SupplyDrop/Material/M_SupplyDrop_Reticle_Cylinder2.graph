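# Translucent, unlit, two-sided material for the supply drop cylinder reticle, built for mesh particles as an animated scanline glow.
# Panning cloud noise and a light mask distort scrolling scanline UVs, which a vertical gradient shapes.
# A camera-vector cubemap adds to that pattern, tinted by EmissiveColor and particle color for emissive.
# An alpha chain adds pixel-depth camera fading and particle alpha, while a separate branch builds a dynamic-parameter pan/contrast mask.
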Material /Game/Gadgets/Assets/SupplyDrop/Material/M_SupplyDrop_Reticle_Cylinder2
graph {
 properties {
  bUsedWithMeshParticles=true
  bUsedWithNiagaraMeshParticles=true
  blendMode=BLEND_Translucent
  pixelDepthOffsetMode=PDOM_Legacy
  shadingModel=MSM_Unlit
  twoSided=true
  usedShadingModels=Unlit
 }
 nodes [
  n1 = ParticleColor()
  n2 = VectorParameter() {parameterName=EmissiveColor, defaultValue=<rgba 0.0227 0.91 1.5 1>}
  n3 = LinearInterpolate(Alpha=MaterialFunctionCall() {materialFunction=&LinearGradient}.VGradient) {constA=-0.25}
  n4 = Panner(Coordinate=TextureCoordinate() {uTiling=4, vTiling=4}) {speedX=-0.125, speedY=0.125}
  n5 = Multiply(
   A=TextureSample() {texture=/Game/Effects/Fort_Effects/Textures/Mask/T_BaseLightMask_Half}.R
   B=Add(A=TextureCoordinate() {uTiling=8, vTiling=2}, B=TextureSample(UVs=@n4) {texture=/Game/Effects/Fort_Effects/Textures/Tile/T_CloudScape_08}.R)
  )
  n6 = TextureCoordinate() {uTiling=5, vTiling=4}
  n7 = TextureSample(UVs=Add(A=Multiply(A=@n5) {constB=0.05}, B=Panner(Coordinate=@n6) {speedX=0.5, speedY=-0.2})) {texture=/Game/Effects/Fort_Effects/Textures/Tile/T_ScanLine_Analog}
  n8 = Add(
   A=TextureSampleParameterCube!(UVs=CameraVectorWS(), Apply View MipBias) {parameterName=Cubemap, texture=/Game/CharClasses/Technician/Gadgets/BASE/Effects/T_CrystalColor_Tile}.RGB
   B=LinearInterpolate(Alpha=Multiply(A=@n3, B=@n7.R)) {constB=2}
  )
  MaterialFunctionCall(Emissive=Multiply(A=@n1.RGB, B=Multiply(A=@n2.RGB, B=@n8))) {materialFunction=&MF_FX_EmissiveAdjustments}
  n9 = Multiply(A=@n3, B=Multiply(A=@n7.RGB) {constB=0.125})
  n10 = Multiply(A=@n9, B=PixelDepth().R)
  n11 = Constant()
  n12 = ScalarParameter() {defaultValue=32, parameterName=CameraFadeDistance}
  n13 = Multiply(A=@n9, B=OneMinus(None=SphereMask(A=@n10, B=@n11, Radius=@n12) {attenuationRadius=128, hardnessPercent=10}))
  Multiply(A=@n1.A, B=@n13)
  n14 = DynamicParameter() {desc=0 to -.5 for contrast control, defaultValue=<rgba 0 0 0 1>, paramNames [PanMask, "Contrast (-.5)", Param3, Param4]}
  n15 = Add(A=OneMinus(None=TextureCoordinate()), B=@n14.PanMask) {constB=0, desc=0 to -1 to pan Mask}
  Multiply(A=@n13, B=LinearInterpolate(A=@n14."Contrast (-.5)", Alpha=ComponentMask(None=@n15) {r=true}) {constA=-0.5}) {constB=-1}
 ]
}
