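# Translucent unlit beam-trail material for the Supply Drop gadget's trail, with responsive AA enabled.
# A smoke trail texture pans along Offset-shifted UV1, warped by a separately panning tiled noise.
# Its red channel, tinted by particle color and masked by green, feeds an emissive adjustment function.
# Opacity combines texture alpha, a ribbon ramp and the green mask, scaled by particle alpha and a Fade parameter.
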Material /Game/Gadgets/Assets/SupplyDrop/Material/M_SupplyDrop_Trail
graph {
 properties {
  bEnableResponsiveAA=true
  bUsedWithBeamTrails=true
  blendMode=BLEND_Translucent
  pixelDepthOffsetMode=PDOM_Legacy
  shadingModel=MSM_Unlit
  usedShadingModels=Unlit
 }
 nodes [
  n1 = ParticleColor()
  n2 = Add(A=TextureCoordinate() {coordinateIndex=1}, B=AppendVector(A=ScalarParameter() {parameterName=Offset}, B=Constant()))
  n3 = TextureCoordinate() {coordinateIndex=1, uTiling=2}
  n4 = Multiply(
   A=Constant2Vector() {g=4, r=0.2}
   B=TextureSample(UVs=Panner(Coordinate=@n3) {speedX=-0.1, speedY=0.12}) {texture=&T_Smooth_tile}.G
  )
  n5 = Constant() {r=0.045}
  n6 = TextureSample(UVs=Add(A=Panner(Coordinate=@n2) {speedX=-0.06}, B=Multiply(A=@n4, B=@n5))) {texture=/Game/Effects/Fort_Effects/Textures/Smoke/T_Lobber_Trail}
  n7 = ComponentMask(None=@n6.R) {r=true}
  MaterialFunctionCall(Emissive=LinearInterpolate(A=Constant(), B=Multiply(A=@n1.RGB, B=@n7), Alpha=ComponentMask(None=@n6.G) {g=true})) {materialFunction=&MF_FX_EmissiveAdjustments}
  n8 = Constant()
  n9 = Multiply(A=@n6.A, B=TextureSample() {texture=/Game/Effects/Fort_Effects/Textures/Ramps/T_Ribbon_Ramp}.G)
  Multiply(A=Multiply(A=@n1.A, B=LinearInterpolate(A=@n8, B=@n9, Alpha=@n6.G)), B=ScalarParameter() {defaultValue=1, parameterName=Fade})
  OneMinus(None=@n7)
 ]
}
