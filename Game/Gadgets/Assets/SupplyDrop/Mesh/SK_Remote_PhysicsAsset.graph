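# Physics asset for the supply drop gadget's remote skeletal mesh, SK_Remote.
# It holds a single body on bone joint1, shaped as one capsule offset and rotated relative to that bone.
# It has no constraints, so the remote is represented as one collision shape with no articulated parts.
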PhysicsAsset /Game/Gadgets/Assets/SupplyDrop/Mesh/SK_Remote_PhysicsAsset
bodies {joint1 [{shapeType=Capsule, length=17.5682, radius=6.0702, center=<16.361 -10.7719 -1.9187>, rotation=<rot 0 15.1095 -86.0386>}]}
constraints []
